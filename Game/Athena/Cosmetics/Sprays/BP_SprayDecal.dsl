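; Replicated spray decal actor that renders a player's cosmetic spray and fades away when its surroundings change.
; Replicated spray info resolves texture or banner icon and color, then async loads feed a decal with a dynamic material and timed fade-in.
; The server binds to overlapping building actors' damage, death, destruction and trap events; these, or bounces, trigger a replicated fade-out and lifespan.
; Creative playlists also get box collision once loading finishes.
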(event EventPostLevelSaveSpawnCleanup
  (|CallParentFunction)
  (CallFunction|UnbindandClearNearbyBuildingActors)
  (CallFunction|FindNearbyBuildingActorsandBindDelegates))

(event Custom|OnWallDestoryed
  (params (DestroyedActor :type Actor%20Object%20Reference))
  (CallFunction|StartSprayFadeOutDuetoNewPlacement))

(event Custom|OnNearbyWallDamaged
  (params (DamagedActor :type Actor%20Object%20Reference) (Damage :type Float%20%28single-precision%29) (InstigatedBy :type Controller%20Object%20Reference) (DamageCauser :type Actor%20Object%20Reference) (HitLocation :type Vector) (FHitComponent :type Primitive%20Component%20Object%20Reference) (BoneName :type Name) (Momentum :type Vector))
  (bind _asfort_player_controller (Utilities|Casting|CastToFortPlayerController InstigatedBy)
    (:then
      (if (or (Utilities|Equal(Object) _asfort_player_controller (Variables|Default|GetSpawningPlayerController)) (Class|FortPlayerController|HasPermissiontoEditWorld _asfort_player_controller self HitLocation (Math|Rotator|MakeRotator 0.0)))
        _))
    (:CastFailed
      _))
  (CallFunction|StartSprayFadeOutDuetoNewPlacement))

(event Custom|OnNearbyWallDied
  (params (DamagedActor :type Actor%20Object%20Reference) (Damage :type Float%20%28single-precision%29) (InstigatedBy :type Controller%20Object%20Reference) (DamageCauser :type Actor%20Object%20Reference) (HitLocation :type Vector) (FHitComponent :type Primitive%20Component%20Object%20Reference) (BoneName :type Name) (Momentum :type Vector))
  (CallFunction|StartSprayFadeOutDuetoNewPlacement))

(event Custom|OnNearbyTrapPlaced
  (params (Trap :type Building%20Trap%20Object%20Reference) (bDetatched :type Boolean))
  (CallFunction|StartSprayFadeOutDuetoNewPlacement))

(event EventStartSprayFadeOutDueToNewPlacement
  (Utilities|FlowControl|DoOnce false
    (:Completed
      (Variables|Default|SetDecalFadeoutTime 1.0)
      (Networking|SwitchHasAuthority
        (:Authority
          (Actor|SetLifeSpan (Variables|Default|GetDecalFadeoutTime))
          (CallFunction|UnbindandClearNearbyBuildingActors))))))

(event EventBeginPlay
  (bind _spawningplayercontroller (Variables|Default|GetSpawningPlayerController))
  (bind _sprayasset (Variables|Default|GetSprayAsset))
  (Utilities|FlowControl|Sequence
    (:then_0
      ;[1 1 1] Make the editor preview display snappier (the small delay is normally used to line up well with the spray montage)
      (if (Variables|Default|GetbIsFrontEndPreview)
        (Variables|Default|SetSpawnStartTimeDelay 0.0)))
    (:then_1
      (Networking|SwitchHasAuthority
        (:Authority
          (if (and (Utilities|IsValid _spawningplayercontroller :!class /Script/Engine.KismetSystemLibrary) (Utilities|IsValid _sprayasset :!class /Script/Engine.KismetSystemLibrary))
            (CallFunction|SetSprayParameters _spawningplayercontroller _sprayasset))
          (CallFunction|FindNearbyBuildingActorsandBindDelegates))))))

(event Custom|OnSprayAssetReplicatedDown
  (bind _decaltextureoverridesoft (Variables|Default|GetDecalTextureOverrideSoft))
  (bind _returnvalue (Asset|IsAssetNull _decaltextureoverridesoft))
  (Utilities|FlowControl|Sequence
    (:then_0
      (Variables|Default|SetLoadsOutstanding (select _returnvalue 1 2))
      (if (not _returnvalue)
        (bind _object (Utilities|AsyncLoadAsset (Variables|Default|GetDecalTextureOverrideSoft))
          (:Completed
            (Variables|Default|SetDecalTextureOverride (Utilities|Casting|CastToTexture2D _object))))))
    (:then_1
      (bind _object_2 (Utilities|AsyncLoadAsset (Spray|GetDecalMaterial (Variables|Default|GetSprayAsset)))
        (:Completed
          (Variables|Default|SetEmissiveMatSource (Utilities|Casting|CastToMaterialInterface _object_2))))))
  ;[1 1 1] Are we done async loading?
  (Math|Integer|DecrementInt (Variables|Default|GetLoadsOutstanding)
    (:%20%20
      (if (== (Variables|Default|GetLoadsOutstanding) 0)
        (CallFunction|CreateDecalComponents)
        ;[1 0.01949 0] Test for creative only after the data is ready
        (bind _gamestate (Class|FortAsyncActionCurrentPlaylistReady|PlaylistDataReadyAsync)
          (:OnReady
            (if (Class|FortGameStateAthena|IsCreativeModePlaylist _gamestate)
              (CallFunction|AddBoxCollisionforCreative))))))))

(event EventOnBounceOccurs
  (if (Variables|FortSprayDecalInstance|GetbDestroyOnNearbyBounce)
    (CallFunction|StartSprayFadeOutDuetoNewPlacement)))

(event EventOnSprayInfoReady
  (bind _sprayasset (Variables|Default|GetSprayAsset))
  (bind _returnvalue (Spray|ShouldUseBannerasTexture _sprayasset))
  (bind _sprayinfo_sprayasset (Utilities|Struct|BreakFortSprayDecalRepPayload (Variables|GetSprayInfo_SprayAsset)))
  (bind (_sprayinfo_sprayasset_2 _sprayinfo_bannername) (Utilities|Struct|BreakFortSprayDecalRepPayload (Variables|GetSprayInfo_SprayAsset)))
  (bind (_sprayinfo_sprayasset_3 _sprayinfo_bannername_2 _sprayinfo_bannercolor) (Utilities|Struct|BreakFortSprayDecalRepPayload (Variables|GetSprayInfo_SprayAsset)))
  (Variables|Default|SetSprayAsset (Utilities|ResolveSoftReference _sprayinfo_sprayasset))
  (Utilities|FlowControl|Sequence
    (:then_0
      (Variables|Default|SetDecalTextureOverrideSoft (Spray|GetDecalTexture (Variables|Default|GetSprayAsset))))
    (:then_1
      (if (and (Utilities|Name|NotEqual(Name) _sprayinfo_bannername "None") _returnvalue)
        (bind _returnvalue_2 (Banner|GetBannerLargeIconAssetfromRowName _sprayinfo_bannername "Spray"))
        (Variables|Default|SetDecalTextureOverrideSoft _returnvalue_2)))
    (:then_2
      (if (and _returnvalue (Utilities|Name|NotEqual(Name) _sprayinfo_bannercolor "None"))
        (bind _returnvalue_3 (Banner|GetBannerColorfromRowName _sprayinfo_bannercolor))
        (Variables|Default|SetBannerPrimaryColor (Banner|GetPrimaryColor _returnvalue_3))
        (Variables|Default|SetBannerSecondaryColor (Banner|GetSecondaryColor _returnvalue_3))))
    (:then_3
      (CallFunction|OnSprayAssetReplicatedDown))))

(fn AddBoxCollisionForCreative
  (AddComponent|NODE_AddBoxComponent-0 :RelativeTransform (Math|Transform|MakeTransform "0, 0, 0" "0, 0, 0" (+ "10.000000, 0,0.000000" (Math|Vector|MakeVector 0.0 (* (Variables|Default|GetDecalSize) 0.9) (* (Variables|Default|GetDecalSize) 0.9))))))

(fn CreateDecalComponents
  (bind (_sprayinfo_sprayasset _sprayinfo_bannername _sprayinfo_bannercolor _sprayinfo_savedstatvalue) (Utilities|Struct|BreakFortSprayDecalRepPayload (Variables|GetSprayInfo_SprayAsset)))
  (bind _emissivedecal (Variables|Default|GetEmissiveDecal))
  (bind (_sprayinfo_sprayasset_2 _sprayinfo_bannername_2 _sprayinfo_bannercolor_2 _sprayinfo_savedstatvalue_2 _sprayinfo_sortorder) (Utilities|Struct|BreakFortSprayDecalRepPayload (Variables|GetSprayInfo_SprayAsset)))
  (bind _returnvalue (AddComponent|NODE_AddDecalComponent-6))
  (Variables|Default|SetEmissiveDecal _returnvalue)
  ;[1 1 1] Create MID (always used for start time, even if not doing a banner or other texture override)
  (bind _returnvalue_2 (Rendering|Material|CreateDynamicMaterialInstance (Variables|Default|GetEmissiveMatSource) "None" "None" :!class /Script/Engine.KismetMaterialLibrary))
  (Variables|Default|SetEmissiveDecalMID _returnvalue_2)
  (Rendering|Components|Decal|SetDecalMaterial _emissivedecal (Variables|Default|GetEmissiveDecalMID))
  (Transformation|SetRelativeScale3D _emissivedecal (Math|Vector|MakeVector (Variables|Default|GetDecalDepth) (Variables|Default|GetDecalSize) (Variables|Default|GetDecalSize)))
  (Utilities|FlowControl|Sequence
    (:then_0
      ;[1 1 1] Set decal texture if any
      (Utilities|IsValid (Variables|Default|GetDecalTextureOverride)
        (:Is%20Valid
          (Rendering|Material|SetTextureParameterValue (Variables|Default|GetEmissiveDecalMID) "DecalTexture" (Variables|Default|GetDecalTextureOverride)))))
    (:then_1
      ;[1 1 1] set banner color if any
      (if (Spray|ShouldUseBannerasTexture (Variables|Default|GetSprayAsset))
        (Rendering|Material|SetVectorParameterValue (Variables|Default|GetEmissiveDecalMID) "ColorLight" (Variables|Default|GetBannerPrimaryColor) :!class /Script/Engine.MaterialInstanceDynamic)))
    (:then_2
      ;[1 1 1] Set progressive stat param
      (Rendering|Material|SetScalarParameterValue (Variables|Default|GetEmissiveDecalMID) "SavedStatValue" (Math|Conversions|ToFloat(Integer) _sprayinfo_savedstatvalue) :!class /Script/Engine.MaterialInstanceDynamic))
    (:then_3
      ;[1 1 1] Spawn start time (used for fade in animation)
      (Rendering|Material|SetScalarParameterValue (Variables|Default|GetEmissiveDecalMID) "DecalSpawnTime" (+ (Utilities|Time|GetRealTimeSeconds) (Variables|Default|GetSpawnStartTimeDelay)) :!class /Script/Engine.MaterialInstanceDynamic))
    (:then_4
      ;[1 1 1] Set the sort order based on rep data
      (Rendering|Components|Decal|SetSortOrder (Variables|Default|GetEmissiveDecal) _sprayinfo_sortorder))))

(fn FindNearbyBuildingActorsAndBindDelegates
  (bind _outactors (Collision|BoxOverlapActors (Transformation|GetActorLocation) (Math|Conversions|ToVector(Float) (/ (Variables|Default|GetDecalSize) 2.0)) (Utilities|Array|MakeArray "ObjectTypeQuery1" "ObjectTypeQuery2") "/Script/FortniteGame.BuildingActor" (Utilities|Array|MakeArray self) :!class /Script/Engine.KismetSystemLibrary))
  (bind _output_get (Variables|Default|SetNearbyBuildingActors _outactors))
  (for _array_element _output_get
    (bind _asbuilding_actor (Utilities|Casting|CastToBuildingActor _array_element))
    (Utilities|Casting|CastToBP_SprayDecal _array_element
      (:CastFailed
        (Utilities|FlowControl|Sequence
          (:then_0
            ;[1 1 1] get damage / death notifies on server for any nearby walls
            (if (Variables|FortSprayDecalInstance|GetbDestroyOnNearbyDestruction)
              (EventDispatchers|BindEventtoOnDied _asbuilding_actor (EventDispatchers|CreateEvent :!event OnNearbyWallDied))))
          (:then_1
            (if (Variables|FortSprayDecalInstance|GetbDestroyOnNearbyDamageTaken)
              (EventDispatchers|BindEventtoOnDamaged _asbuilding_actor (EventDispatchers|CreateEvent :!event OnNearbyWallDamaged))))
          (:then_2
            (if (Variables|Default|GetbDestroyOnTrapPlaced)
              (bind _asbuilding_static_mesh (Utilities|Casting|CastToBuildingSMActor _asbuilding_actor))
              (Building|BindEventtoOnTrapPlacementChanged _asbuilding_static_mesh (EventDispatchers|CreateEvent :!event OnNearbyTrapPlaced))))
          (:then_3
            (Game|BindEventtoOnDestroyed _asbuilding_actor (EventDispatchers|CreateEvent :!event OnWallDestoryed))
            ;[1 1 1] Don't send these in creative mode until we have a better system
            (bind _cancompletequest (Class|AthenaQuestFunctionLibrary|AthenaCheckCreativeMode (Game|GetGameState)))
            (if _cancompletequest
              _)))))))

(fn GetInstigatorPlayerId
  (return (Default|GetGameAccountId (Variables|Default|GetSpawningPlayerController))))

(fn InitializeSpraySpawnedfromGameplay
  (params (InDecalSize :type Float%20%28double-precision%29) (InSprayAsset :type Athena%20Spray%20Item%20Definition%20Object%20Reference) (InSpawningPlayerController :type Fort%20Player%20Controller%20Object%20Reference))
  (bind _spawningplayercontroller (Variables|Default|GetSpawningPlayerController))
  (bind _sprayasset (Variables|Default|GetSprayAsset))
  (Variables|Default|SetDecalSize InDecalSize)
  (Variables|Default|SetSprayAsset InSprayAsset)
  (Variables|Default|SetSpawningPlayerController InSpawningPlayerController)
  (Variables|Default|SetbIsFrontEndPreview false)
  (Networking|SwitchHasAuthority
    (:Authority
      (if (and (Utilities|IsValid _spawningplayercontroller :!class /Script/Engine.KismetSystemLibrary) (Utilities|IsValid _sprayasset :!class /Script/Engine.KismetSystemLibrary))
        (CallFunction|SetSprayParameters _spawningplayercontroller _sprayasset)
        (else
          _))
      (CallFunction|FindNearbyBuildingActorsandBindDelegates))))

(fn OnRep_DecalFadeoutTime
  (Utilities|IsValid (Variables|Default|GetEmissiveDecal)
    (:Is%20Valid
      (if (> (Variables|Default|GetDecalFadeoutTime) 0.0)
        (Rendering|Components|Decal|SetFadeOut (Variables|Default|GetEmissiveDecal) 0.0 (Variables|Default|GetDecalFadeoutTime) false)))))

(fn UnbindAndClearNearbyBuildingActors
  (for _array_element (Variables|Default|GetNearbyBuildingActors)
    (Game|UnbindEventfromOnDestroyed _array_element (EventDispatchers|CreateEvent :!event OnWallDestoryed))
    (bind _asbuilding_actor (Utilities|Casting|CastToBuildingActor _array_element))
    (EventDispatchers|UnbindEventfromOnDied _asbuilding_actor (EventDispatchers|CreateEvent :!event OnNearbyWallDied))
    (EventDispatchers|UnbindEventfromOnDamaged _asbuilding_actor (EventDispatchers|CreateEvent :!event OnNearbyWallDamaged))
    (bind _asbuilding_static_mesh (Utilities|Casting|CastToBuildingSMActor _asbuilding_actor))
    (Building|UnbindEventfromOnTrapPlacementChanged _asbuilding_static_mesh (EventDispatchers|CreateEvent :!event OnNearbyTrapPlaced)))
  (Utilities|Array|Clear (Variables|Default|GetNearbyBuildingActors)))

(fn ConstructionScript)
