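; Preview display actor for Athena spray cosmetics, showing the selected spray as a decal.
; PreviewSpray stores the item, plays a 2D preview cue, silently kills any previous decal and spawns BP_SprayDecal attached to SpawnArea.
; PerformSpecialAction repeats that sequence with the stored spray, and both end in a two-second delay with nothing after it.
; On destruction it kills the decal and fades out the preview audio.
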(event Visuals|EventPerformSpecialAction
  (params (ActionName :type Name :default "None"))
  (bind _mydecalactor (Variables|Default|GetMyDecalActor))
  (bind _returnvalue (Audio|SpawnSound2D "/Game/Athena/Sounds/Sprays/Athena_Spray_Preview_Cue.Athena_Spray_Preview_Cue"))
  (bind _output_get (Variables|Default|SetSpray%20Preview%20Audio _returnvalue))
  (Audio|Components|Audio|Play _output_get 0.0)
  (Utilities|IsValid _mydecalactor
    (:Is%20Valid
      (Gameplay|SilentDie _mydecalactor false :!class /Script/FortniteGame.BuildingActor))
    (:Is%20Not%20Valid
      _))
  (bind _output_get_2 (Variables|Default|SetDecalSpawnPoint (Math|Vector|MakeVector 0.0)))
  (bind _returnvalue_2 (Game|SpawnActorfromClass "/Game/Athena/Cosmetics/Sprays/BP_SprayDecal.BP_SprayDecal_C" (Math|Transform|MakeTransform _output_get_2 "0,90.000000, 0") "AlwaysSpawn" "MultiplyWithRoot" 0 (Variables|Default|GetDecalSize) (Variables|Default|GetMySpray) (Utilities|Array|Get(acopy) (Fortnite|GetLocalFortPlayerControllers true)) true 0 0 0))
  (bind _output_get_3 (Variables|Default|SetMyDecalActor _returnvalue_2))
  (Transformation|AttachActorToComponent _output_get_3 (Variables|Default|GetSpawnArea) "None" "KeepRelative" "KeepRelative" "SnapToTarget" true)
  (Utilities|FlowControl|Delay 2.0))

(event Spray|EventPreviewSpray
  (params (Spray :type Athena%20Spray%20Item%20Definition%20Object%20Reference))
  (bind _mydecalactor (Variables|Default|GetMyDecalActor))
  (Variables|Default|SetMySpray Spray)
  (bind _returnvalue (Audio|SpawnSound2D "/Game/Athena/Sounds/Sprays/Athena_Spray_Preview_Cue.Athena_Spray_Preview_Cue"))
  (bind _output_get (Variables|Default|SetSpray%20Preview%20Audio _returnvalue))
  (Audio|Components|Audio|Play _output_get 0.0)
  (Utilities|IsValid _mydecalactor
    (:Is%20Valid
      (Gameplay|SilentDie _mydecalactor false :!class /Script/FortniteGame.BuildingActor))
    (:Is%20Not%20Valid
      _))
  (bind _output_get_2 (Variables|Default|SetDecalSpawnPoint (Math|Vector|MakeVector 0.0)))
  (bind _returnvalue_2 (Game|SpawnActorfromClass "/Game/Athena/Cosmetics/Sprays/BP_SprayDecal.BP_SprayDecal_C" (Math|Transform|MakeTransform _output_get_2 "0,90.000000, 0") "AlwaysSpawn" "MultiplyWithRoot" 0 (Variables|Default|GetDecalSize) (Variables|Default|GetMySpray) (Utilities|Array|Get(acopy) (Fortnite|GetLocalFortPlayerControllers true)) true 0 0 0))
  (bind _output_get_3 (Variables|Default|SetMyDecalActor _returnvalue_2))
  (Transformation|AttachActorToComponent _output_get_3 (Variables|Default|GetSpawnArea) "None" "KeepRelative" "KeepRelative" "SnapToTarget" true)
  (Utilities|FlowControl|Delay 2.0))

(event EventDestroyed
  (Gameplay|SilentDie (Variables|Default|GetMyDecalActor) false :!class /Script/FortniteGame.BuildingActor)
  (|CallParentFunction)
  (bind _spray_preview_audio (Variables|Default|GetSpray%20Preview%20Audio)
    (:then
      (Audio|Components|Audio|FadeOut _spray_preview_audio 0.3 0.0 "Linear" :!class /Script/Engine.AudioComponent))
    (:else)))

(fn ConstructionScript)
