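# Master alpha-composite deferred decal material for Athena spray cosmetics.
# A static or flipbook texture blends dark and light colors, with time-of-day tinting on one color branch and an additive Creative highlight grid.
# Opacity combines texture alpha or green, a facing mask and an optional panning, masked shine boost.
# Spawn time and decal lifetime drive a noise-texture dissolve for fading in and out, and color is premultiplied by that opacity.
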Material /Game/Athena/Cosmetics/Sprays/M_SprayDecal_Master
graph {
 properties {blendMode=BLEND_AlphaComposite, materialDomain=MD_DeferredDecal, pixelDepthOffsetMode=PDOM_Legacy, usedShadingModels=Default Lit}
 nodes [
  Constant()
  Constant()
  n1 = Constant()
  MaterialFunctionCall(Emissive=ShadingPathSwitch(Default=@n1, Mobile=@n1) {desc=Migrated from FeatureLevelSwitch}) {materialFunction=&MF_FX_EmissiveAdjustments}
  n2 = TextureSample(UVs=TextureCoordinate() {uTiling=1.9, vTiling=1.9}) {texture=/Game/Weapons/Prototype/ObjectMover/Textures/T_Movetool_Transition}
  n3 = VectorParameter() {parameterName=GridColor, defaultValue=<rgba 0.5344 0.1706 0.0578 1>}
  n4 = Multiply(
   A=Add(A=Multiply(A=@n2.R, B=@n3.RGB), B=VectorParameter() {parameterName=FillColor, defaultValue=<rgba 0.1927 0.0677 0.0204 1>}.RGB)
   B=ScalarParameter() {parameterName=ShowCreativeHighlight}
  )
  n5 = MaterialFunctionCall(
   Color_Morning=VectorParameter() {parameterName=Morning, defaultValue=<rgba 0.1771 0.0891 0.0577 1>}.RGB
   Color_Day=VectorParameter() {parameterName=Day, defaultValue=<rgba 0.6458 0.6458 0.6458 1>}.RGB
   Color_Evening=VectorParameter() {parameterName=Evening, defaultValue=<rgba 0.0781 0.0435 0.0375 1>}.RGB
   Color_Night=VectorParameter() {parameterName=Night, defaultValue=<rgba 0.1182 0.2527 0.3698 1>}.RGB
  ) {materialFunction=&Color_Time_of_Day}
  n6 = TextureObjectParameter() {parameterName=DecalTexture, texture=/Game/2dAssets/Sprays/Season8/Textures/T_Spray_S8_Gold}
  n7 = MaterialFunctionCall(
   "Animation  Phase (0-1)"=Multiply(A=Time(), B=ScalarParameter() {defaultValue=0.1, parameterName=FlipbookSpeed})
   Number of Rows=ScalarParameter() {defaultValue=8, parameterName=FlipbookRows}
   Number of Columns=ScalarParameter() {defaultValue=8, parameterName=FlipbookColumns}
   Texture=@n6
  ) {materialFunction=&FlipBook}
  n8 = TextureSample(Tex=@n6) {texture=/Game/Weapons/Prototype/SprayGun/Materials/T_TestDecal}
  n9 = StaticBoolParameter() {parameterName=UseFlipbook}
  n10 = LinearInterpolate(
   A=VectorParameter() {parameterName=ColorDark, defaultValue=<rgba 0 0 0 1>}.RGB
   B=VectorParameter() {parameterName=ColorLight, defaultValue=<rgba 1 1 1 1>}.RGB
   Alpha=StaticSwitch(True=@n7.Result, False=@n8.RGB, Value=@n9)
  )
  n11 = StaticBoolParameter() {parameterName=UseGreenAsAlpha}
  n12 = StaticSwitch(
   True=MaterialFunctionCall(Color=@n7.Result) {materialFunction=&SplitComponents}.G
   False=@n7.Alpha
   Value=@n11
  )
  n13 = StaticSwitch(True=@n8.G, False=@n8.G, Value=@n11)
  n14 = Multiply(
   A=Multiply(A=StaticSwitch(True=@n12, False=@n13, Value=@n9), B=ScalarParameter() {defaultValue=1, parameterName=Opacity})
   B=MaterialFunctionCall() {materialFunction=/Game/Weapons/Prototype/SprayGun/Materials/MF_DecalFacingMask}.Result
  )
  n15 = Multiply(A=@n14, B=ScalarParameter() {defaultValue=100, parameterName=ShineIntensity}) {constB=100}
  n16 = Panner(
   Coordinate=Rotator(Coordinate=TextureCoordinate(), Time=ScalarParameter() {defaultValue=-2, parameterName=Shine Rotation}) {speed=8}
   Speed=AppendVector(A=ScalarParameter() {defaultValue=0.5, parameterName=x speed}, B=ScalarParameter() {parameterName=y speed})
  )
  n17 = Abs(None=Add(A=Frac(None=@n16)) {constB=-0.5})
  n18 = Multiply(
   A=Power(Base=ComponentMask(None=@n17) {r=true}) {constExponent=4}
   B=TextureSampleParameter2D() {parameterName=ShineMask, texture=/Game/Weapons/WeaponSkins/Wraps/Textures/T_SimpleBlack}.R
  )
  n19 = Subtract(A=Time() {bIgnorePause=true}, B=ScalarParameter() {parameterName=DecalSpawnTime})
  n20 = Constant() {desc=1 / Fade in duration, r=1.25}
  n21 = Min(A=DecalLifetimeOpacity().Opacity, B=Saturate(None=Multiply(A=@n19, B=@n20)))
  n22 = Add(A=TextureSample() {texture=/Game/Global/Textures/Default/Noise/T_SprayTime}.R, B=Add(A=Constant() {r=-1}, B=Multiply(A=@n21) {constB=1.1}))
  n23 = Multiply(
   A=StaticSwitchParameter(True=LinearInterpolate(A=@n14, B=@n15, Alpha=@n18), False=@n14) {parameterName=Shine}
   B=Clamp(None=Multiply(A=@n22) {constB=10})
  )
  Multiply(A=Add(A=@n4, B=Multiply(A=@n5.Result, B=@n10)), B=@n23)
  ShadingPathSwitch(Default=Multiply(A=Add(A=@n4, B=@n10), B=@n23)) {desc=Migrated from FeatureLevelSwitch}
 ]
}
